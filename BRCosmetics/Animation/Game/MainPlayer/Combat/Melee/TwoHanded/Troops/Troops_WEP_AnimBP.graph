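# Animation Blueprint for the Troops two-handed pickaxe weapon mesh, with logic to animate its saw-tooth chain during melee swings.
# Outside the frontend, each update caches the pickaxe, owning pawn and player AnimBP, reads the melee-playing flag, and interpolates a chain play-rate scale.
# The final pose is only DefaultSlot over the ref pose.
# The Start/Loop/End state machine, per-bone blend and rate-scaled loop player are not connected to it.
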AnimBlueprint /BRCosmetics/Animation/Game/MainPlayer/Combat/Melee/TwoHanded/Troops/Troops_WEP_AnimBP
target_skeleton=/Game/Weapons/FORT_Melee/Pickaxe_Troops/Meshes/Troops_Pickaxe_Skeleton
graphs [
 AnimGraph {nodes [
   n1 = Local Space Ref Pose()
   AnimGraphNode_StateMachine_0 = StateMachine-OtherOptionForChainMovement.() {is=title AnimGraphNode_StateMachine}
   Blend Poses by bool(
    BlendPose_0=Use cached pose 'MontagePose'()
    BlendPose_1=Blend Poses by bool(
     BlendPose_0=Layered blend per bone(BasePose=@n1, BlendPoses_0=@AnimGraphNode_StateMachine_0) {layerSetup [{branchFilters [{boneName=dyn_sawTooth_0}]}]}
     BlendPose_1=Use cached pose 'MontagePose'()
     BlendTime_0=0
     BlendTime_1=0
     bActiveValue=true
    ) {blendType=Linear}
    BlendTime_0=0
    BlendTime_1=0
    bActiveValue=Get IsFrontEnd() -> (IsFrontEnd:bool)
   ) {blendType=Linear}
   Output Pose(Result=Slot 'DefaultSlot'(Source=Local Space Ref Pose())) {layerGroup=None}
   Sequence Player 'Troops_PickaxeChain_Loop_WEP'!(PlayRate:float=float * float(A=Get ChainPlayRate() -> (ChainPlayRate:double), B=Get ChainPlayRateScaleSmooth() -> (ChainPlayRateScaleSmooth:double)) {is=title K2Node_PromotableOperator}) -> (Pose) {sequence=/Game/Animation/Game/MainPlayer/Combat/Melee/TwoHanded/Troops/Troops_PickaxeChain_Loop_WEP}
   OR Boolean(A=Get ShouldChainPlay() -> (ShouldChainPlay:bool), B=Get Is Playing Melee Anim() -> (Is Playing Melee Anim:bool))
  ]}
 StateMachine-OtherOptionForChainMovement. {
  entry=RefPose
  path=AnimGraph.AnimGraphNode_StateMachine_0.StateMachine-OtherOptionForChainMovement.
  links [
   {
    dst=RefPose
    rule=AnimGraph.AnimGraphNode_StateMachine_0.StateMachine-OtherOptionForChainMovement..AnimStateTransitionNode_3.Transition
    src=End
    via=AnimStateTransitionNode_3
   }
   {
    dst=End
    rule=AnimGraph.AnimGraphNode_StateMachine_0.StateMachine-OtherOptionForChainMovement..AnimStateTransitionNode_2.Transition
    src=Loop
    via=AnimStateTransitionNode_2
   }
   {
    dst=Start
    rule=AnimGraph.AnimGraphNode_StateMachine_0.StateMachine-OtherOptionForChainMovement..AnimStateTransitionNode_0.Transition
    src=RefPose
    via=AnimStateTransitionNode_0
   }
   {
    dst=Loop
    rule=AnimGraph.AnimGraphNode_StateMachine_0.StateMachine-OtherOptionForChainMovement..AnimStateTransitionNode_1.Transition
    src=Start
    via=AnimStateTransitionNode_1
   }
  ]
  nodes []
 }
 RefPose {
  path=AnimGraph.AnimGraphNode_StateMachine_0.StateMachine-OtherOptionForChainMovement..AnimStateNode_0.RefPose
  nodes [Output Animation Pose(Result=Local Space Ref Pose())]
 }
 Start {
  path=AnimGraph.AnimGraphNode_StateMachine_0.StateMachine-OtherOptionForChainMovement..AnimStateNode_1.Start
  nodes [Output Animation Pose(Result=Sequence Player 'Troops_PickaxeChain_Into_WEP'!(bLoopAnimation:bool=False) -> (Pose) {bLoopAnimation=false, sequence=/Game/Animation/Game/MainPlayer/Combat/Melee/TwoHanded/Troops/Troops_PickaxeChain_Into_WEP})]
 }
 Loop {
  path=AnimGraph.AnimGraphNode_StateMachine_0.StateMachine-OtherOptionForChainMovement..AnimStateNode_2.Loop
  nodes [Output Animation Pose(Result=Sequence Player 'Troops_PickaxeChain_Loop_WEP'!(PlayRate=Get ChainPlayRate(), bLoopAnimation:bool=Get Is Playing Melee Anim()) -> (Pose) {bLoopAnimation=false, sequence=&Troops_PickaxeChain_Loop_WEP})]
 }
 End {
  path=AnimGraph.AnimGraphNode_StateMachine_0.StateMachine-OtherOptionForChainMovement..AnimStateNode_3.End
  nodes [Output Animation Pose(Result=Sequence Player 'Troops_PickaxeChain_Outro_WEP'!(PlayRate:float=Get ChainPlayRate(), bLoopAnimation:bool=False) -> (Pose) {bLoopAnimation=false, sequence=/Game/Animation/Game/MainPlayer/Combat/Melee/TwoHanded/Troops/Troops_PickaxeChain_Outro_WEP})]
 }
 Transition {
  path=AnimGraph.AnimGraphNode_StateMachine_0.StateMachine-OtherOptionForChainMovement..AnimStateTransitionNode_0.Transition
  nodes [Result(bCanEnterTransition=Get Is Playing Melee Anim())]
 }
 Transition {path=AnimGraph.AnimGraphNode_StateMachine_0.StateMachine-OtherOptionForChainMovement..AnimStateTransitionNode_1.Transition, nodes [Result()]}
 Transition {
  path=AnimGraph.AnimGraphNode_StateMachine_0.StateMachine-OtherOptionForChainMovement..AnimStateTransitionNode_2.Transition
  nodes [
   n1 = Get Is Playing Melee Anim()
   n2 = "Get Relevant Anim Time Remaining (Loop)"!() -> (ReturnValue:float) {is=title K2Node_AnimGetter}
   Result(bCanEnterTransition=AND Boolean(A=NOT Boolean(A=@n1), B="Equal (Float)"(A=@n2)))
  ]
 }
 Transition {path=AnimGraph.AnimGraphNode_StateMachine_0.StateMachine-OtherOptionForChainMovement..AnimStateTransitionNode_3.Transition, nodes [Result()]}
 EventGraph {nodes [
   Set Delta Time X(Delta Time X=Event BlueprintUpdateAnimation().DeltaTimeX)
   > n1 = Set IsFrontEnd(IsFrontEnd:bool=IsCosmeticObjectDisplayedInFrontendContext!(CosmeticObject:Object Reference) -> (ReturnValue:bool)) -> (Output_Get:bool)
   > n2 = Branch(Condition=@n1)
   K2Node_Composite_1 = InGame(execute=@n2.else) {is=title K2Node_Composite}
  ]}
 InGame {
  path=EventGraph.K2Node_Composite_1.InGame
  nodes [
   AnimNotify_ExitedEndState() {is=title K2Node_Event}
   > Set ShouldChainPlay(ShouldChainPlay:bool=false) -> (Output_Get:bool)
   Inputs()
   > n1 = Branch(Condition=IsNullItem!(ItemObject:Object Reference=Get As B Athena Pickaxe Troops() -> (As B Athena Pickaxe Troops:B Athena Pickaxe Troops Object Reference)) -> (Is Null Item:bool))
   n2 = GetOwningActor()
   > Is Valid(InputObject=@n2)
   > n3 = Cast To B_Athena_Pickaxe_Troops!(execute, Object=@n2) -> (then, CastFailed, AsB Athena Pickaxe Troops:B Athena Pickaxe Troops Object Reference) {is=title K2Node_DynamicCast}
   > n4 = Set As B Athena Pickaxe Troops(As B Athena Pickaxe Troops:B Athena Pickaxe Troops Object Reference=@n3) -> (Output_Get:B Athena Pickaxe Troops Object Reference)
   n5 = GetOwner(self:Actor Object Reference=Get As B Athena Pickaxe Troops())
   Is Valid(exec=[@n16.then, @n4.then], InputObject=@n5)
   > n6 = Cast To FortPlayerPawn(Object=@n5)
   > Set As Fort Player Pawn(As Fort Player Pawn:Fort Player Pawn Object Reference=@n6) -> (Output_Get:Fort Player Pawn Object Reference)
   n7 = GetAnimInstance(self=Get Mesh(self=Get As Fort Player Pawn() -> (As Fort Player Pawn:Fort Player Pawn Object Reference)))
   > n8 = Is Valid(InputObject=@n7)
   n9 = Cast To Fortnite_M_Avg_Player_AnimBlueprint!(execute=[@n17.then, @n8.Is Valid], Object=@n7) -> (then, CastFailed, AsFortnite M Avg Player Anim Blueprint:Fortnite M Avg Player Anim Blueprint Object Reference) {is=title K2Node_DynamicCast}
   > n10 = Set As Fortnite M Avg Player Anim Blueprint(As Fortnite M Avg Player Anim Blueprint:Fortnite M Avg Player Anim Blueprint Object Reference=@n9) -> (Output_Get:Fortnite M Avg Player Anim Blueprint Object Reference)
   n11 = Set Is Playing Melee Anim(
    execute=[@n17.else, @n10.then]
    Is Playing Melee Anim:bool=Get bIsPlayingMeleeAnim!(self:Fort Player Anim Instance Object Reference=Get As Fortnite M Avg Player Anim Blueprint() -> (As Fortnite M Avg Player Anim Blueprint:Fortnite M Avg Player Anim Blueprint Object Reference)) -> (bIsPlayingMeleeAnim)
   ) -> (Output_Get:bool)
   > n12 = Branch(Condition=@n11)
   > Set ShouldChainPlay(ShouldChainPlay=true)
   > n13 = Set ChainPlayRateScale(ChainPlayRateScale:double=1) -> (Output_Get:double)
   n14 = Set ChainPlayRateScaleSmooth(
    execute=[@n15.then, @n13.then]
    ChainPlayRateScaleSmooth:double=FInterpTo(
     Current=Get ChainPlayRateScaleLastTick() -> (ChainPlayRateScaleLastTick:double)
     Target=Get ChainPlayRateScale() -> (ChainPlayRateScale:double)
     DeltaTime=Get Delta Time X()
     InterpSpeed=17
    )
   ) -> (Output_Get:double)
   > Set ChainPlayRateScaleLastTick(ChainPlayRateScaleLastTick:double=@n14) -> (Output_Get:double)
   n15 = Set ChainPlayRateScale(execute=@n12.else, ChainPlayRateScale=0)
   n16 = Branch(execute=@n1.else, Condition=IsNullItem(ItemObject=Get As Fort Player Pawn()))
   n17 = Branch(execute=@n16.else, Condition=IsNullItem(ItemObject=Get As Fortnite M Avg Player Anim Blueprint()))
   Outputs()
  ]
 }
 OnNewFunction {nodes [OnNewFunction!() -> (then, Context:"Anim Update Context Structure (by ref)", Node:"Anim Node Reference Structure (by ref)") {is=title K2Node_FunctionEntry}]}
]
variables [
 As B Athena Pickaxe Troops {type=B_Athena_Pickaxe_Troops_C, default=null}
 As Fort Player Pawn {type=FortPlayerPawn}
 As Fortnite M Avg Player Anim Blueprint {type=Fortnite_M_Avg_Player_AnimBlueprint_C, default=null}
 ChainPlayRate {type=number, default=50}
 ChainPlayRateScale {type=number}
 ChainPlayRateScaleLastTick {type=number}
 ChainPlayRateScaleSmooth {type=number}
 Delta Time X {type=number}
 Is Playing Melee Anim {type=boolean}
 IsFrontEnd {type=boolean}
 ShouldChainPlay {type=boolean}
]
